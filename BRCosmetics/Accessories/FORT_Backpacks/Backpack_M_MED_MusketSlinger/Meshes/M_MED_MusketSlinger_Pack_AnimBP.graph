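# Animation Blueprint for the MusketSlinger backpack that hides the pack while its matching glider is open.
# Each update follows the owning Fortnite pawn to its parachute mesh and flags a MusketSlinger glider by wildcard-matching the mesh name.
# That flag and the parachute-open state drive a bool alpha on a Modify Bone node that blends the pack bone's scale to zero.
# The glider mesh variable is unused by the graphs.
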AnimBlueprint /BRCosmetics/Accessories/FORT_Backpacks/Backpack_M_MED_MusketSlinger/Meshes/M_MED_MusketSlinger_Pack_AnimBP
target_skeleton=&/M_MED_MusketSlinger_Pack_Skeleton
graphs [
 AnimGraph {nodes [
   n1 = "Transform (Modify) Bone - Bone: pack"!(
    ComponentPose=Local To Component(LocalPose=Input Pose())
    bAlphaBoolEnabled:bool=Get BackpackTransformAlpha() -> (BackpackTransformAlpha:bool)
    Translation="0,0,0"
    Rotation="0,0,0"
    Scale="0,0,0"
   ) -> (Pose) {
    alphaInputType=Bool
    scaleMode=BMM_Additive
    alphaBoolBlend {blendInTime=0.2, blendOption=ExpInOut, blendOutTime=0.3}
    boneToModify {boneName=pack}
    scale=<0 0 0>
   }
   Output Pose(Result=Component To Local(ComponentPose=@n1)) {layerGroup=None}
  ]}
 EventGraph {nodes [
   Event BlueprintUpdateAnimation()
   > n1 = Sequence()
   > n2 = Cast To FortPlayerPawn(Object=GetOwningActor())
   n3 = Get ParachuteAttachment!(self:Fort Player Pawn Object Reference=@n2) -> (ParachuteAttachment:Fort Player Parachute Object Reference)
   > Is Valid(InputObject=@n3)
   n4 = Get ParachuteMesh!(self:Fort Player Parachute Object Reference=@n3) -> (ParachuteMesh:Skeletal Mesh Component Budgeted Object Reference)
   > Is Valid(InputObject=@n4)
   n5 = Get SkeletalMesh!(self:Skinned Mesh Component Object Reference=@n4) -> (SkeletalMesh:Skeletal Mesh Object Reference)
   > Is Valid(InputObject=@n5)
   > Set IsMusketSlingerGlider(IsMusketSlingerGlider:bool=MatchesWildcard(SourceString=GetDisplayName(Object=@n5), Wildcard=MusketSlinger_Glider, SearchCase=CaseSensitive)) -> (Output_Get:bool)
   n6 = Branch(execute=@n1.then_1, Condition=AND Boolean(A=Get bIsParachuteOpened(), B=Get IsMusketSlingerGlider() -> (IsMusketSlingerGlider:bool)))
   > Set BackpackTransformAlpha(BackpackTransformAlpha:bool=true) -> (Output_Get:bool)
   Set BackpackTransformAlpha(execute=@n6.else, BackpackTransformAlpha=false)
   TryGetPawnOwner()
  ]}
]
variables [
 BackpackTransformAlpha {type=boolean}
 IsMusketSlingerGlider {type=boolean}
 MusketSlingerGlider {type=SkeletalMesh, default=/BRCosmetics/Gadgets/Assets/VinderTech_GliderChute/Glider_MusketSlinger/Meshes/MusketSlinger_Glider}
]
